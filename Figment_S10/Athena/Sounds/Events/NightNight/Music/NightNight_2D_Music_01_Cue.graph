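# Sound cue that provides the first 2D music track for the NightNight event.
# A single wave player feeds Event_NightNight_Music_01 straight to the output, with no looping, mixing or randomization.
# It plays slightly boosted through the extra-high-priority events sound class.
# Custom attenuation and listener-focus values are stored, but the attenuation override is not switched on and no attenuation asset is assigned.
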SoundCue /Figment_S10/Athena/Sounds/Events/NightNight/Music/NightNight_2D_Music_01_Cue
root=WavePlayer_1
nodes [{id=WavePlayer_1, properties {soundWaveAssetPtr=/Figment_S10/Athena/Sounds/Events/NightNight/SoundWaves/Event_NightNight_Music_01}}]
properties {
 maxDistance=2097152
 soundClassObject=/Figment_S10/Athena/Sounds/Events/SoundClasses/Fort_Events_Priority_ExtHigh
 volumeMultiplier=1.25
 attenuationOverrides {
  bEnableFocusInterpolation=true
  bEnableListenerFocus=true
  falloffDistance=100000
  focusAttackInterpSpeed=0.5
  focusAzimuth=5
  focusReleaseInterpSpeed=0.25
  lPFFrequencyAtMax=500
  lPFRadiusMax=2000
  lPFRadiusMin=1000
  nonFocusAzimuth=30
  nonFocusVolumeAttenuation=0
  stereoSpread=2048
  attenuationShapeExtents {x=1024}
  customAttenuationCurve {editorCurveData {keys [<at 0.1 0>,<at 0.25 1>,<at 1 0>]}}
 }
}
